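; Looping gameplay cue for the ProtoCannon charge-up, playing audio and particle feedback on the owning Fortnite pawn.
; On application it attaches a start sound flagged for local viewing and times the minimum charge, which also feeds the particle's ChargeTime.
; When the timer fires with the cue tag still present, it plays a charge-met sound and watches that tag.
; Tag loss, removal and EndPlay fade that sound; removal also clears the timer and fades the start sound.
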(event EventBeginPlay)

(event Collision|EventActorBeginOverlap
  (params (OtherActor :type Actor%20Object%20Reference)))

(event EventTick
  (params (DeltaSeconds :type Float%20%28single-precision%29)))

(event EventOnApplication(EitherVFX-preferthisone)
  (params (MyTarget :type Actor%20Object%20Reference) (Parameters :type Gameplay%20Cue%20Parameters%20Structure%20%28by%20ref%29) (ParticleComponents :type Array%20of%20FXSystem%20Component%20Object%20References) (AudioComponents :type Array%20of%20Audio%20Component%20Object%20References) (BurstCameraShakeInstance :type Camera%20Shake%20Base%20Object%20Reference) (BurstDecalInstance :type Decal%20Actor%20Object%20Reference))
  (Variables|Default|SetRemoved false)
  ;[1 1 1] Audio logic and is local parameter failsafe
  (bind _asfort_player_pawn_athena (Utilities|Casting|CastToFortPlayerPawnAthena MyTarget)
    (:then
      (Variables|Default|SetOwningPawn _asfort_player_pawn_athena)
      (bind _returnvalue (Audio|SpawnSoundAttached :Sound (Variables|Default|GetChargeStartSound) :AttachToComponent (Variables|/Script/Engine.Actor|GetRootComponent _asfort_player_pawn_athena) :AttachPointName "None" :LocationType "KeepRelativeOffset" :bStopWhenAttachedToDestroyed true :VolumeMultiplier 1.0 :PitchMultiplier 1.0 :StartTime 0.0 :bAutoDestroy true))
      (bind _output_get (Variables|Default|SetChargeStartAudioComponent _returnvalue))
      (Audio|Parameter|SetBooleanParameter _output_get "IsLocal" (Pawn|IsLocallyViewed _asfort_player_pawn_athena) :!class /Script/AudioExtensions.AudioParameterControllerInterface))
    (:CastFailed
      _))
  (Variables|Default|SetParameters Parameters)
  (bind _returnvalue_2 (Utilities|Time|SetTimerbyEvent (delegate Custom|DoOnMinChargeMet) (ScalableFloat|GetValueatLevel (Variables|Default|GetMinChargeTime)) false true))
  (Variables|Default|SetChargeMetTimerHandle _returnvalue_2))

(event Custom|DoOnMinChargeMet
  (if (and (not (Variables|Default|GetRemoved)) (GameplayTags|HasMatchingGameplayTag (Variables|Default|GetOwningPawn) (CallFunction|GetGameplayCueTag)))
    (bind _returnvalue (Audio|SpawnSoundAttached :Sound (Variables|Default|GetChargeMetSound) :AttachToComponent (Variables|/Script/Engine.Actor|GetRootComponent (Variables|Default|GetOwningPawn)) :AttachPointName "None" :LocationType "KeepRelativeOffset" :bStopWhenAttachedToDestroyed true :VolumeMultiplier 1.0 :PitchMultiplier 1.0 :StartTime 0.0 :bAutoDestroy true))
    (Variables|Default|SetChargeMetAudioComponent _returnvalue)
    (GameplayTags|BindEventWrappertoGameplayTagChanged (Ability|GetFortAbilitySystemComponent (Variables|Default|GetOwningPawn)) (CallFunction|GetGameplayCueTag) (delegate Custom|HandleTagRemoved) false)))

(event EventOnLoopingStart(EitherVFX-preferthisversion)
  (params (MyTarget :type Actor%20Object%20Reference) (Parameters :type Gameplay%20Cue%20Parameters%20Structure%20%28by%20ref%29) (ParticleComponents :type Array%20of%20FXSystem%20Component%20Object%20References) (AudioComponents :type Array%20of%20Audio%20Component%20Object%20References))
  (Variables|Default|SetLoopingAudioComponents AudioComponents)
  (if (Utilities|Array|IsValidIndex ParticleComponents)
    (Effects|Components|ParticleSystem|SetFloatParameter (Utilities|Array|Get(acopy) ParticleComponents) "ChargeTime" (ScalableFloat|GetValueatLevel (Variables|Default|GetMinChargeTime)))))

(event Custom|HandleTagRemoved
  (params (Tag :type Gameplay%20Tag%20Structure%20%28by%20ref%29) (TagCount :type Integer))
  (if (== TagCount 0)
    (CallFunction|FadeOutChargeMetSound)))

(event EventOnRemoval(EitherVFX-preferthisversion)
  (params (MyTarget :type Actor%20Object%20Reference) (Parameters :type Gameplay%20Cue%20Parameters%20Structure%20%28by%20ref%29) (ParticleComponents :type Array%20of%20FXSystem%20Component%20Object%20References) (AudioComponents :type Array%20of%20Audio%20Component%20Object%20References) (BurstCameraShakeInstance :type Camera%20Shake%20Base%20Object%20Reference) (BurstDecalInstance :type Decal%20Actor%20Object%20Reference))
  (Variables|Default|SetRemoved true)
  (Utilities|FlowControl|Sequence
    (:then_0
      (Utilities|Time|ClearandInvalidateTimerbyHandle self (Variables|Default|GetChargeMetTimerHandle)))
    (:then_1
      (bind _chargestartaudiocomponent (Variables|Default|GetChargeStartAudioComponent)
        (:then
          (Audio|Components|Audio|FadeOut _chargestartaudiocomponent 0.5 0.0 "Linear" :!class /Script/Engine.AudioComponent))
        (:else)))
    (:then_2
      (CallFunction|FadeOutChargeMetSound))))

(event EventEndPlay
  (params (EndPlayReason :type EEndPlayReason%20Enum :default "Destroyed"))
  (|CallParentFunction EndPlayReason)
  (CallFunction|FadeOutChargeMetSound))

(event Custom|FadeOutChargeMetSound
  (bind _chargemetaudiocomponent (Variables|Default|GetChargeMetAudioComponent)
    (:then
      (Audio|Components|Audio|FadeOut _chargemetaudiocomponent 0.5 0.0 "Linear" :!class /Script/Engine.AudioComponent))
    (:else)))

(fn ConstructionScript)
